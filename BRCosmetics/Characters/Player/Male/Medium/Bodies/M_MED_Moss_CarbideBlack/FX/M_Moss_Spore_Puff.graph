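# Translucent unlit Niagara sprite material for a moss spore puff effect on a player character cosmetic.
# A spore alpha texture, randomly rotated per particle and churned by four-way chaos motion, is sharpened by an S-curve that grows toward the edges and over particle life.
# A wavy, squared radial mask shapes the puff, with optional automatic age fade, camera fade and depth fade.
# Particle color drives emissive through a shared adjustment function and also scales opacity.
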Material /BRCosmetics/Characters/Player/Male/Medium/Bodies/M_MED_Moss_CarbideBlack/FX/M_Moss_Spore_Puff
graph {
 properties {bUsedWithNiagaraSprites=true, blendMode=BLEND_Translucent, pixelDepthOffsetMode=PDOM_Legacy, shadingModel=MSM_Unlit, usedShadingModels=Unlit}
 nodes [
  n1 = OneMinus(None=ParticleRelativeTime())
  n2 = Subtract(A=TextureCoordinate()) {constB=0.5}
  n3 = ScalarParameter() {defaultValue=0.7, parameterName=UV Scale}
  n4 = Rotator(Coordinate=Add(A=Multiply(A=@n2, B=@n3)) {constB=0.5}, Time=Multiply(A=ParticleRandom()) {constB=10}) {speed=-1}
  n5 = ScalarParameter() {defaultValue=2, parameterName=Speed, sliderMax=9}
  n6 = TextureObject() {samplerType=SAMPLERTYPE_Alpha, texture=/Game/Characters/Player/Male/Medium/Bodies/M_MED_Moss_CarbideBlack/FX/T_Moss_Spores_A}
  n7 = ScalarParameter() {defaultValue=2, parameterName=Scurve, sliderMax=9}
  n8 = Subtract(A=TextureCoordinate()) {constB=0.5}
  n9 = MaterialFunctionCall!(UV=Add(A=Multiply(A=@n8) {constB=1.2}) {constB=0.5}, U Frequency, V Frequency, U Amplitude, V Amplitude, U Speed, V Speed, U Time, V Time) {materialFunction=/Game/Effects/Fort_Effects/MaterialFunctions/MF_WaveUV}
  n10 = Constant() {r=0.5}
  n11 = Saturate(None=Divide(A=Distance(A=@n9.Wave UVs, B=@n10), B=@n10))
  n12 = Max(A=@n11, B=ParticleRelativeTime())
  n13 = OneMinus(None=@n11)
  n14 = Multiply(
   A=MaterialFunctionCall!(
    In=MaterialFunctionCall!(Coordinates=@n4, Speed=@n5, Divisor, Texture=@n6) {materialFunction=/Engine/Functions/Engine_MaterialFunctions03/ComplexMotion/Motion_4WayChaos}.Result
    Power=Multiply(A=@n7, B=@n12)
   ) {materialFunction=/Game/Characters/Player/Male/Medium/Bodies/M_MED_Moss_CarbideBlack/FX/Functions/MF_SCurve_Float}.Result
   B=Multiply(A=@n13, B=@n13)
  )
  n15 = StaticSwitchParameter(True=Multiply(A=Multiply(A=@n1, B=@n1), B=@n14), False=@n14) {parameterName=Fade out automatically}
  n16 = Saturate(None=MaterialFunctionCall(Fade Length=Constant() {r=200}, Fade Offset=Constant() {r=40}) {materialFunction=&CameraDepthFade}.Result)
  n17 = ParticleColor()
  DepthFade(
   Opacity=Multiply(A=Multiply(A=@n15, B=@n16), B=@n17.A)
   FadeDistance=ScalarParameter() {defaultValue=7, parameterName=Depth Fade Distance, sliderMax=66}
  )
  MaterialFunctionCall(Emissive=@n17.RGB) {materialFunction=&MF_FX_EmissiveAdjustments}
 ]
}
