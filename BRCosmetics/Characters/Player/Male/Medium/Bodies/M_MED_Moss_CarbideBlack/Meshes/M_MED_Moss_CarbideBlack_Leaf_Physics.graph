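# Physics asset for the leaf mesh of the Moss CarbideBlack medium male player body cosmetic, defining bodies for its dangling foliage bones.
# Most leaves are boxes jointed to small capsule "_MSTR" anchors with mostly limited swing and twist; butt, scapula, back-leg and some foot pieces form chains or branches.
# Anchors connect with free angular motion to the pelvis, or to the right thigh bodies for right foot and back-leg foliage.
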PhysicsAsset /BRCosmetics/Characters/Player/Male/Medium/Bodies/M_MED_Moss_CarbideBlack/Meshes/M_MED_Moss_CarbideBlack_Leaf_Physics
bodies {
 dyn_l_arm_folliage_01 [{shapeType=Box, extentX=9.1, extentY=9.1, extentZ=3.1, center=<3.6993 -0.0001 0>}]
 dyn_l_arm_folliage_01_MSTR [{shapeType=Capsule, length=0.505, radius=0.505}]
 dyn_l_arm_folliage_02 [{shapeType=Box, extentX=9.1, extentY=9.1, extentZ=3.1, center=<3.6993 -0.0001 0>}]
 dyn_l_arm_folliage_02_MSTR [{shapeType=Capsule, length=0.505, radius=0.505}]
 dyn_l_arm_folliage_03 [{shapeType=Box, extentX=9.1, extentY=9.1, extentZ=3.1, center=<3.6993 -0.0001 0>}]
 dyn_l_arm_folliage_03_MSTR [{shapeType=Capsule, length=0.505, radius=0.505}]
 dyn_l_arm_folliage_04 [{shapeType=Box, extentX=9.1, extentY=9.1, extentZ=3.1, center=<3.6993 -0.0001 0>}]
 dyn_l_arm_folliage_04_MSTR [{shapeType=Capsule, length=0.505, radius=0.505}]
 dyn_l_arm_folliage_05 [{shapeType=Box, extentX=9.1, extentY=9.1, extentZ=3.1, center=<3.6993 -0.0001 0>}]
 dyn_l_arm_folliage_05_MSTR [{shapeType=Capsule, length=0.505, radius=0.505}]
 dyn_l_arm_folliage_06 [{shapeType=Box, extentX=9.1, extentY=9.1, extentZ=3.1, center=<1.2474 -0.0001 0>}]
 dyn_l_arm_folliage_06_MSTR [{shapeType=Capsule, length=0.505, radius=0.505}]
 dyn_l_arm_folliage_07 [{shapeType=Box, extentX=6.0166, extentY=9.1, extentZ=1.777, center=<3.0676 -0.0001 0>}]
 dyn_l_arm_folliage_07_MSTR [{shapeType=Capsule, length=0.505, radius=0.505, rotation=<rot 0 -90 0>}]
 dyn_l_butt_folliage_01 [{shapeType=Box, extentX=4.1369, extentY=9.1, extentZ=1.9189, center=<2.2198 0 0>, rotation=<rot -0.0001 0 6.1841>}]
 dyn_l_butt_folliage_02 [{shapeType=Box, extentX=3.8685, extentY=9.1, extentZ=1.4037, center=<1.9305 -0.0001 0>}]
 dyn_l_butt_folliage_03 [{shapeType=Box, extentX=4.7268, extentY=9.1, extentZ=1.6275, center=<2.5221 -0.0001 0>}]
 dyn_l_butt_folliage_MSTR [{shapeType=Capsule, length=0.505, radius=0.505}]
 dyn_l_chest_Folliage [{shapeType=Box, extentX=6, extentY=6, extentZ=1.5, center=<0 0 -1.9084>}]
 dyn_l_chest_Folliage_MSTR [{shapeType=Capsule, length=0.5, radius=0.5}]
 dyn_l_foot_folliage_01 [{shapeType=Box, extentX=9.1, extentY=9.1, extentZ=3.1, center=<3.6993 -0.0001 0>}]
 dyn_l_foot_folliage_01_MSTR [{shapeType=Capsule, length=0.505, radius=0.505}]
 dyn_l_foot_folliage_02 [{shapeType=Box, extentX=9.1, extentY=9.1, extentZ=3.1, center=<3.6993 -0.0001 0>}]
 dyn_l_foot_folliage_03 [{shapeType=Box, extentX=9.1, extentY=9.1, extentZ=3.1, center=<3.6993 -0.0001 0>}]
 dyn_l_leg_folliage_01 [{shapeType=Box, extentX=9.1, extentY=9.1, extentZ=3.1, center=<3.6993 -0.0001 0>}]
 dyn_l_leg_folliage_01_MSTR [{shapeType=Capsule, length=0.505, radius=0.505}]
 dyn_l_leg_folliage_02 [{shapeType=Box, extentX=9.1, extentY=9.1, extentZ=3.1, center=<1.9109 -0.0003 -2.0084>}]
 dyn_l_leg_folliage_02_MSTR [{shapeType=Capsule, length=0.505, radius=0.505}]
 dyn_l_leg_folliage_03 [{shapeType=Box, extentX=9.1, extentY=9.1, extentZ=3.1, center=<2.4219 -0.0002 -1.3688>}]
 dyn_l_leg_folliage_03_MSTR [{shapeType=Capsule, length=0.505, radius=0.505}]
 dyn_l_leg_folliage_04 [{shapeType=Box, extentX=9.1, extentY=9.1, extentZ=3.1, center=<3.6993 -0.0001 0>}]
 dyn_l_leg_folliage_04_MSTR [{shapeType=Capsule, length=0.505, radius=0.505}]
 dyn_l_leg_folliage_05 [{shapeType=Box, extentX=9.1, extentY=9.1, extentZ=3.1, center=<3.6993 -0.0001 0>}]
 dyn_l_leg_folliage_05_MSTR [{shapeType=Capsule, length=0.505, radius=0.505}]
 dyn_l_leg_folliage_06 [{shapeType=Box, extentX=9.1, extentY=9.1, extentZ=3.1, center=<3.6993 -0.0001 0>}]
 dyn_l_leg_folliage_06_MSTR [{shapeType=Capsule, length=0.505, radius=0.505}]
 dyn_l_leg_folliage_07 [{shapeType=Box, extentX=9.1, extentY=9.1, extentZ=3.1, center=<3.6993 -0.0001 0>}]
 dyn_l_leg_folliage_07_MSTR [{shapeType=Capsule, length=0.505, radius=0.505}]
 dyn_l_neck_Folliage [{shapeType=Box, extentX=6, extentY=6, extentZ=1.5, center=<0 0 -1.4695>}]
 dyn_l_neck_Folliage_MSTR [{shapeType=Capsule, length=0.505, radius=0.505, rotation=<rot 0 -90 0>}]
 dyn_l_scapula_Folliage_01 [{shapeType=Capsule, length=1, radius=1, center=<-0.383 0 0>, rotation=<rot 90 0 0>}]
 dyn_l_scapula_Folliage_02 [{shapeType=Capsule, length=1, radius=1, center=<-0.383 0 0>, rotation=<rot 90 0 0>}]
 dyn_l_scapula_Folliage_03 [{shapeType=Capsule, length=1, radius=1, center=<-0.383 0 0>, rotation=<rot 90 0 0>}]
 dyn_l_scapula_Folliage_Base [{shapeType=Capsule, length=1.1047, radius=0.3276}]
 dyn_l_shoulder_Folliage_a [{shapeType=Box, extentX=9.1, extentY=9.1, extentZ=3.1, center=<0 0 -0.5136>}]
 dyn_l_shoulder_Folliage_a_MSTR [{shapeType=Capsule, length=0.505, radius=0.505}]
 dyn_l_shoulder_Folliage_b [{shapeType=Box, extentX=9.1, extentY=9.1, extentZ=3.1, center=<0 0 -0.5136>}]
 dyn_l_shoulder_Folliage_b_MSTR [{shapeType=Capsule, length=0.505, radius=0.505}]
 dyn_l_shoulder_Folliage_c [{shapeType=Box, extentX=9.1, extentY=9.1, extentZ=3.1, center=<0 0 -0.5136>}]
 dyn_l_shoulder_Folliage_c_MSTR [{shapeType=Capsule, length=0.505, radius=0.505}]
 dyn_l_shoulder_Folliage_d [{shapeType=Box, extentX=5.7586, extentY=6.3577, extentZ=2, center=<0 0.0001 -1.0023>}]
 dyn_l_shoulder_Folliage_d_MSTR [{shapeType=Capsule, length=0.505, radius=0.505}]
 dyn_l_shoulder_Folliage_e [{shapeType=Box, extentX=5.7586, extentY=6.3577, extentZ=2, center=<0 0.0001 -1.0023>}]
 dyn_l_shoulder_Folliage_e_MSTR [{shapeType=Capsule, length=0.505, radius=0.505}]
 dyn_l_torso_folliage_01 [{shapeType=Box, extentX=9.1, extentY=9.1, extentZ=3.1, center=<3.6993 -0.0001 0>}]
 dyn_l_torso_folliage_01_MSTR [{shapeType=Capsule, length=0.505, radius=0.505}]
 dyn_l_torso_folliage_02 [{shapeType=Box, extentX=9.1, extentY=9.1, extentZ=3.1, center=<-0.2165 -0.0002 -2.8133>, rotation=<rot -11.8677 0 0>}]
 dyn_l_torso_folliage_02_MSTR [{shapeType=Capsule, length=0.505, radius=0.505}]
 dyn_l_torso_folliage_03 [{shapeType=Box, extentX=9.1, extentY=9.1, extentZ=3.1, center=<3.6993 -0.0001 0>}]
 dyn_l_torso_folliage_03_MSTR [{shapeType=Capsule, length=0.505, radius=0.505}]
 dyn_l_torso_folliage_04 [{shapeType=Box, extentX=9.1, extentY=9.1, extentZ=3.1, center=<3.6993 -0.0001 0>}]
 dyn_l_torso_folliage_04_MSTR [{shapeType=Capsule, length=0.505, radius=0.505}]
 dyn_l_torso_folliage_05 [{shapeType=Box, extentX=9.1, extentY=9.1, extentZ=3.1, center=<4.308 -0.0001 0>}]
 dyn_l_torso_folliage_05_MSTR [{shapeType=Capsule, length=0.505, radius=0.505}]
 dyn_l_torso_folliage_06 [{shapeType=Box, extentX=9.1, extentY=9.1, extentZ=3.1, center=<3.6993 -0.0001 0>}]
 dyn_l_torso_folliage_06_MSTR [{shapeType=Capsule, length=0.505, radius=0.505}]
 dyn_r_arm_folliage_01 [{shapeType=Box, extentX=9.1, extentY=9.1, extentZ=3.1, center=<3.6993 -0.0001 0>}]
 dyn_r_arm_folliage_01_MSTR [{shapeType=Capsule, length=0.505, radius=0.505}]
 dyn_r_arm_folliage_02 [{shapeType=Box, extentX=9.1, extentY=9.1, extentZ=3.1, center=<3.6993 -0.0001 0>}]
 dyn_r_arm_folliage_02_MSTR [{shapeType=Capsule, length=0.505, radius=0.505}]
 dyn_r_arm_folliage_03 [{shapeType=Box, extentX=9.1, extentY=9.1, extentZ=3.1, center=<0.5296 -0.0001 -0.8771>}]
 dyn_r_arm_folliage_03_MSTR [{shapeType=Capsule, length=0.505, radius=0.505}]
 dyn_r_arm_folliage_05 [{shapeType=Box, extentX=7.2259, extentY=5.7663, extentZ=2.1016, center=<3.6993 -0.0001 0>}]
 dyn_r_arm_folliage_05_MSTR [{shapeType=Capsule, length=0.505, radius=0.505}]
 dyn_r_arm_folliage_06 [{shapeType=Box, extentX=9.1, extentY=9.1, extentZ=3.1, center=<3.6993 -0.0001 0>}]
 dyn_r_arm_folliage_06_MSTR [{shapeType=Capsule, length=0.505, radius=0.505}]
 dyn_r_arm_folliage_07 [{shapeType=Box, extentX=9.1, extentY=9.1, extentZ=3.1, center=<3.6993 -0.0001 0>}]
 dyn_r_arm_folliage_07_MSTR [{shapeType=Capsule, length=0.505, radius=0.505}]
 dyn_r_foot_folliage_01 [{shapeType=Box, extentX=9.1, extentY=9.1, extentZ=3.1, center=<3.6993 -0.0001 0>}]
 dyn_r_foot_folliage_01_MSTR [{shapeType=Capsule, length=0.505, radius=0.505}]
 dyn_r_foot_folliage_02 [{shapeType=Box, extentX=9.1, extentY=9.1, extentZ=3.1, center=<3.6993 -0.0001 0>}]
 dyn_r_foot_folliage_03 [{shapeType=Box, extentX=9.1, extentY=9.1, extentZ=3.1, center=<3.6993 -0.0001 0>}]
 dyn_r_foot_folliage_03_MSTR [{shapeType=Capsule, length=0.505, radius=0.505}]
 dyn_r_foot_folliage_04 [{shapeType=Box, extentX=9.1, extentY=9.1, extentZ=3.1, center=<3.6993 -0.0001 0>}]
 dyn_r_foot_folliage_05 [{shapeType=Capsule, length=3.8053, radius=2.2559, center=<3.8418 0 0>, rotation=<rot -90 0 0>}]
 dyn_r_foot_folliage_05_MSTR [{shapeType=Capsule, length=0.505, radius=0.505}]
 dyn_r_leg_back_folliage_01 [{shapeType=Box, extentX=4.9646, extentY=5.8722, extentZ=2.6611, center=<2.2252 0 0.0001>, rotation=<rot -11.194 0 -5.1629>}]
 dyn_r_leg_back_folliage_02 [{shapeType=Box, extentX=6.3134, extentY=7.9452, extentZ=2.6516, center=<2.5252 0 0>, rotation=<rot 0 0 -6.6251>}]
 dyn_r_leg_back_folliage_MSTR [{shapeType=Capsule, length=0.505, radius=0.505}]
 dyn_r_shoulder_folliage_a [{shapeType=Box, extentX=9.1, extentY=9.1, extentZ=3.1, center=<0 0 -0.2742>}]
 dyn_r_shoulder_folliage_a_MSTR [{shapeType=Capsule, length=0.505, radius=0.505}]
 dyn_r_shoulder_folliage_b [{shapeType=Box, extentX=9.1, extentY=9.1, extentZ=3.1, center=<0 -0.3327 -0.8035>, rotation=<rot 3.876 1.5168 21.3911>}]
 dyn_r_shoulder_folliage_b_MSTR [{shapeType=Capsule, length=0.505, radius=0.505}]
 dyn_r_shoulder_folliage_c [{shapeType=Box, extentX=9.1, extentY=9.1, extentZ=3.1, center=<0 0 -0.9962>}]
 dyn_r_shoulder_folliage_c_MSTR [{shapeType=Capsule, length=0.505, radius=0.505}]
 dyn_r_shoulder_folliage_d [{shapeType=Box, extentX=9.1, extentY=9.1, extentZ=3.1, center=<0 0 -0.6188>}]
 dyn_r_shoulder_folliage_d_MSTR [{shapeType=Capsule, length=1.4671, radius=1.1343}]
 pelvis [{shapeType=Capsule, length=5.9657, radius=11.4899}]
 thigh_r [{shapeType=Capsule, length=17.8398, radius=8.1571, center=<17.3549 0 0>, rotation=<rot -90 0 0>}]
 thigh_twist_01_r [{shapeType=Capsule, length=2, radius=2}]
}
constraints [
 {
  bone1Name=dyn_l_arm_folliage_05_MSTR
  bone2Name=pelvis
  swing1LimitDegrees=45
  swing1Motion=Free
  swing2LimitDegrees=45
  swing2Motion=Free
  twistLimitDegrees=45
  twistMotion=Free
 }
 {
  bone1Name=dyn_l_arm_folliage_03_MSTR
  bone2Name=pelvis
  swing1LimitDegrees=45
  swing1Motion=Free
  swing2LimitDegrees=45
  swing2Motion=Free
  twistLimitDegrees=45
  twistMotion=Free
 }
 {
  bone1Name=dyn_l_arm_folliage_04_MSTR
  bone2Name=pelvis
  swing1LimitDegrees=45
  swing1Motion=Free
  swing2LimitDegrees=45
  swing2Motion=Free
  twistLimitDegrees=45
  twistMotion=Free
 }
 {
  bone1Name=dyn_l_arm_folliage_01_MSTR
  bone2Name=pelvis
  swing1LimitDegrees=45
  swing1Motion=Free
  swing2LimitDegrees=45
  swing2Motion=Free
  twistLimitDegrees=45
  twistMotion=Free
 }
 {
  bone1Name=dyn_l_arm_folliage_02_MSTR
  bone2Name=pelvis
  swing1LimitDegrees=45
  swing1Motion=Free
  swing2LimitDegrees=45
  swing2Motion=Free
  twistLimitDegrees=45
  twistMotion=Free
 }
 {
  bone1Name=dyn_l_arm_folliage_06_MSTR
  bone2Name=pelvis
  swing1LimitDegrees=45
  swing1Motion=Free
  swing2LimitDegrees=45
  swing2Motion=Free
  twistLimitDegrees=45
  twistMotion=Free
 }
 {
  bone1Name=dyn_l_shoulder_Folliage_c_MSTR
  bone2Name=pelvis
  swing1LimitDegrees=45
  swing1Motion=Free
  swing2LimitDegrees=45
  swing2Motion=Free
  twistLimitDegrees=45
  twistMotion=Free
 }
 {
  bone1Name=dyn_l_shoulder_Folliage_b_MSTR
  bone2Name=pelvis
  swing1LimitDegrees=45
  swing1Motion=Free
  swing2LimitDegrees=45
  swing2Motion=Free
  twistLimitDegrees=45
  twistMotion=Free
 }
 {
  bone1Name=dyn_l_shoulder_Folliage_a_MSTR
  bone2Name=pelvis
  swing1LimitDegrees=45
  swing1Motion=Free
  swing2LimitDegrees=45
  swing2Motion=Free
  twistLimitDegrees=45
  twistMotion=Free
 }
 {
  bone1Name=dyn_r_arm_folliage_05_MSTR
  bone2Name=pelvis
  swing1LimitDegrees=45
  swing1Motion=Free
  swing2LimitDegrees=45
  swing2Motion=Free
  twistLimitDegrees=45
  twistMotion=Free
 }
 {
  bone1Name=dyn_r_arm_folliage_01_MSTR
  bone2Name=pelvis
  swing1LimitDegrees=45
  swing1Motion=Free
  swing2LimitDegrees=45
  swing2Motion=Free
  twistLimitDegrees=45
  twistMotion=Free
 }
 {
  bone1Name=dyn_r_arm_folliage_02_MSTR
  bone2Name=pelvis
  swing1LimitDegrees=45
  swing1Motion=Free
  swing2LimitDegrees=45
  swing2Motion=Free
  twistLimitDegrees=45
  twistMotion=Free
 }
 {
  bone1Name=dyn_r_arm_folliage_03_MSTR
  bone2Name=pelvis
  swing1LimitDegrees=45
  swing1Motion=Free
  swing2LimitDegrees=45
  swing2Motion=Free
  twistLimitDegrees=45
  twistMotion=Free
 }
 {
  bone1Name=dyn_r_arm_folliage_07_MSTR
  bone2Name=pelvis
  swing1LimitDegrees=45
  swing1Motion=Free
  swing2LimitDegrees=45
  swing2Motion=Free
  twistLimitDegrees=45
  twistMotion=Free
 }
 {
  bone1Name=dyn_r_arm_folliage_06_MSTR
  bone2Name=pelvis
  swing1LimitDegrees=45
  swing1Motion=Free
  swing2LimitDegrees=45
  swing2Motion=Free
  twistLimitDegrees=45
  twistMotion=Free
 }
 {
  bone1Name=dyn_r_shoulder_folliage_c_MSTR
  bone2Name=pelvis
  swing1LimitDegrees=45
  swing1Motion=Free
  swing2LimitDegrees=45
  swing2Motion=Free
  twistLimitDegrees=45
  twistMotion=Free
 }
 {
  bone1Name=dyn_r_shoulder_folliage_a_MSTR
  bone2Name=pelvis
  swing1LimitDegrees=45
  swing1Motion=Free
  swing2LimitDegrees=45
  swing2Motion=Free
  twistLimitDegrees=45
  twistMotion=Free
 }
 {
  bone1Name=dyn_r_shoulder_folliage_b_MSTR
  bone2Name=pelvis
  swing1LimitDegrees=45
  swing1Motion=Free
  swing2LimitDegrees=45
  swing2Motion=Free
  twistLimitDegrees=45
  twistMotion=Free
 }
 {
  bone1Name=dyn_l_torso_folliage_04_MSTR
  bone2Name=pelvis
  swing1LimitDegrees=45
  swing1Motion=Free
  swing2LimitDegrees=45
  swing2Motion=Free
  twistLimitDegrees=45
  twistMotion=Free
 }
 {
  bone1Name=dyn_l_torso_folliage_05_MSTR
  bone2Name=pelvis
  swing1LimitDegrees=45
  swing1Motion=Free
  swing2LimitDegrees=45
  swing2Motion=Free
  twistLimitDegrees=45
  twistMotion=Free
 }
 {
  bone1Name=dyn_l_torso_folliage_06_MSTR
  bone2Name=pelvis
  swing1LimitDegrees=45
  swing1Motion=Free
  swing2LimitDegrees=45
  swing2Motion=Free
  twistLimitDegrees=45
  twistMotion=Free
 }
 {
  bone1Name=dyn_l_torso_folliage_02_MSTR
  bone2Name=pelvis
  swing1LimitDegrees=45
  swing1Motion=Free
  swing2LimitDegrees=45
  swing2Motion=Free
  twistLimitDegrees=45
  twistMotion=Free
 }
 {
  bone1Name=dyn_l_torso_folliage_03_MSTR
  bone2Name=pelvis
  swing1LimitDegrees=45
  swing1Motion=Free
  swing2LimitDegrees=45
  swing2Motion=Free
  twistLimitDegrees=45
  twistMotion=Free
 }
 {
  bone1Name=dyn_l_leg_folliage_07_MSTR
  bone2Name=pelvis
  swing1LimitDegrees=45
  swing1Motion=Free
  swing2LimitDegrees=45
  swing2Motion=Free
  twistLimitDegrees=45
  twistMotion=Free
 }
 {
  bone1Name=dyn_l_leg_folliage_05_MSTR
  bone2Name=pelvis
  swing1LimitDegrees=45
  swing1Motion=Free
  swing2LimitDegrees=45
  swing2Motion=Free
  twistLimitDegrees=45
  twistMotion=Free
 }
 {
  bone1Name=dyn_l_leg_folliage_06_MSTR
  bone2Name=pelvis
  swing1LimitDegrees=45
  swing1Motion=Free
  swing2LimitDegrees=45
  swing2Motion=Free
  twistLimitDegrees=45
  twistMotion=Free
 }
 {
  bone1Name=dyn_l_foot_folliage_01_MSTR
  bone2Name=pelvis
  swing1LimitDegrees=45
  swing1Motion=Free
  swing2LimitDegrees=45
  swing2Motion=Free
  twistLimitDegrees=45
  twistMotion=Free
 }
 {
  bone1Name=dyn_l_leg_folliage_04_MSTR
  bone2Name=pelvis
  swing1LimitDegrees=45
  swing1Motion=Free
  swing2LimitDegrees=45
  swing2Motion=Free
  twistLimitDegrees=45
  twistMotion=Free
 }
 {
  bone1Name=dyn_l_leg_folliage_03_MSTR
  bone2Name=pelvis
  swing1LimitDegrees=45
  swing1Motion=Free
  swing2LimitDegrees=45
  swing2Motion=Free
  twistLimitDegrees=45
  twistMotion=Free
 }
 {
  bone1Name=dyn_l_leg_folliage_02_MSTR
  bone2Name=pelvis
  swing1LimitDegrees=45
  swing1Motion=Free
  swing2LimitDegrees=45
  swing2Motion=Free
  twistLimitDegrees=45
  twistMotion=Free
 }
 {
  bone1Name=dyn_l_leg_folliage_01_MSTR
  bone2Name=pelvis
  swing1LimitDegrees=45
  swing1Motion=Free
  swing2LimitDegrees=45
  swing2Motion=Free
  twistLimitDegrees=45
  twistMotion=Free
 }
 {
  bone1Name=dyn_r_foot_folliage_03_MSTR
  bone2Name=thigh_r
  swing1LimitDegrees=45
  swing1Motion=Free
  swing2LimitDegrees=45
  swing2Motion=Free
  twistLimitDegrees=45
  twistMotion=Free
 }
 {
  bone1Name=dyn_r_foot_folliage_01_MSTR
  bone2Name=thigh_r
  swing1LimitDegrees=45
  swing1Motion=Free
  swing2LimitDegrees=45
  swing2Motion=Free
  twistLimitDegrees=45
  twistMotion=Free
 }
 {
  bone1Name=dyn_l_butt_folliage_MSTR
  bone2Name=pelvis
  swing1LimitDegrees=45
  swing1Motion=Free
  swing2LimitDegrees=45
  swing2Motion=Free
  twistLimitDegrees=45
  twistMotion=Free
 }
 {
  bone1Name=dyn_l_torso_folliage_01_MSTR
  bone2Name=pelvis
  swing1LimitDegrees=45
  swing1Motion=Free
  swing2LimitDegrees=45
  swing2Motion=Free
  twistLimitDegrees=45
  twistMotion=Free
 }
 {bone1Name=dyn_l_arm_folliage_05, bone2Name=dyn_l_arm_folliage_05_MSTR, swing1LimitDegrees=25, swing2LimitDegrees=25, twistLimitDegrees=5}
 {bone1Name=dyn_l_arm_folliage_04, bone2Name=dyn_l_arm_folliage_04_MSTR, swing1LimitDegrees=25, swing2LimitDegrees=25, twistLimitDegrees=5}
 {bone1Name=dyn_l_arm_folliage_03, bone2Name=dyn_l_arm_folliage_03_MSTR, swing1LimitDegrees=25, swing2LimitDegrees=25, twistLimitDegrees=25}
 {bone1Name=dyn_l_arm_folliage_01, bone2Name=dyn_l_arm_folliage_01_MSTR, swing1LimitDegrees=25, swing2LimitDegrees=50, twistLimitDegrees=5}
 {bone1Name=dyn_l_arm_folliage_02, bone2Name=dyn_l_arm_folliage_02_MSTR, swing1LimitDegrees=25, swing2LimitDegrees=25, twistLimitDegrees=5}
 {bone1Name=dyn_l_arm_folliage_06, bone2Name=dyn_l_arm_folliage_06_MSTR, swing1LimitDegrees=45, swing2LimitDegrees=45, twistLimitDegrees=30}
 {
  bone1Name=dyn_l_shoulder_Folliage_c
  bone2Name=dyn_l_shoulder_Folliage_c_MSTR
  swing1LimitDegrees=25
  swing2LimitDegrees=25
  twistLimitDegrees=5
 }
 {
  bone1Name=dyn_l_shoulder_Folliage_b
  bone2Name=dyn_l_shoulder_Folliage_b_MSTR
  swing1LimitDegrees=25
  swing2LimitDegrees=25
  twistLimitDegrees=5
 }
 {
  bone1Name=dyn_l_shoulder_Folliage_a
  bone2Name=dyn_l_shoulder_Folliage_a_MSTR
  swing1LimitDegrees=25
  swing2LimitDegrees=25
  twistLimitDegrees=5
 }
 {bone1Name=dyn_r_arm_folliage_05, bone2Name=dyn_r_arm_folliage_05_MSTR, swing1LimitDegrees=25, swing2LimitDegrees=25, twistLimitDegrees=5}
 {bone1Name=dyn_r_arm_folliage_01, bone2Name=dyn_r_arm_folliage_01_MSTR, swing1LimitDegrees=25, swing2LimitDegrees=25, twistLimitDegrees=5}
 {bone1Name=dyn_r_arm_folliage_02, bone2Name=dyn_r_arm_folliage_02_MSTR, swing1LimitDegrees=25, swing2LimitDegrees=25, twistLimitDegrees=5}
 {bone1Name=dyn_r_arm_folliage_03, bone2Name=dyn_r_arm_folliage_03_MSTR, swing1LimitDegrees=45, swing2LimitDegrees=45, twistLimitDegrees=30}
 {bone1Name=dyn_r_arm_folliage_07, bone2Name=dyn_r_arm_folliage_07_MSTR, swing1LimitDegrees=80, swing2LimitDegrees=80, twistLimitDegrees=5}
 {bone1Name=dyn_r_arm_folliage_06, bone2Name=dyn_r_arm_folliage_06_MSTR, swing1LimitDegrees=80, swing2LimitDegrees=80, twistLimitDegrees=5}
 {
  bone1Name=dyn_r_shoulder_folliage_c
  bone2Name=dyn_r_shoulder_folliage_c_MSTR
  swing1LimitDegrees=25
  swing2LimitDegrees=25
  twistLimitDegrees=5
 }
 {
  bone1Name=dyn_r_shoulder_folliage_a
  bone2Name=dyn_r_shoulder_folliage_a_MSTR
  swing1LimitDegrees=25
  swing2LimitDegrees=25
  twistLimitDegrees=5
 }
 {
  bone1Name=dyn_r_shoulder_folliage_b
  bone2Name=dyn_r_shoulder_folliage_b_MSTR
  swing1LimitDegrees=25
  swing2LimitDegrees=25
  twistLimitDegrees=5
 }
 {
  bone1Name=dyn_l_torso_folliage_04
  bone2Name=dyn_l_torso_folliage_04_MSTR
  swing1LimitDegrees=25
  swing2LimitDegrees=25
  twistLimitDegrees=25
 }
 {
  bone1Name=dyn_l_torso_folliage_05
  bone2Name=dyn_l_torso_folliage_05_MSTR
  swing1LimitDegrees=25
  swing2LimitDegrees=25
  twistLimitDegrees=25
 }
 {
  bone1Name=dyn_l_torso_folliage_06
  bone2Name=dyn_l_torso_folliage_06_MSTR
  swing1LimitDegrees=25
  swing2LimitDegrees=25
  twistLimitDegrees=25
 }
 {
  bone1Name=dyn_l_torso_folliage_02
  bone2Name=dyn_l_torso_folliage_02_MSTR
  swing1LimitDegrees=45
  swing2LimitDegrees=45
  twistLimitDegrees=20
 }
 {
  bone1Name=dyn_l_torso_folliage_03
  bone2Name=dyn_l_torso_folliage_03_MSTR
  swing1LimitDegrees=80
  swing2LimitDegrees=80
  twistLimitDegrees=25
 }
 {bone1Name=dyn_l_leg_folliage_07, bone2Name=dyn_l_leg_folliage_07_MSTR, swing1LimitDegrees=25, swing2LimitDegrees=25, twistLimitDegrees=5}
 {bone1Name=dyn_l_leg_folliage_05, bone2Name=dyn_l_leg_folliage_05_MSTR, swing1LimitDegrees=25, swing2LimitDegrees=25, twistLimitDegrees=5}
 {bone1Name=dyn_l_leg_folliage_06, bone2Name=dyn_l_leg_folliage_06_MSTR, swing1LimitDegrees=25, swing2LimitDegrees=25, twistLimitDegrees=5}
 {bone1Name=dyn_l_foot_folliage_01, bone2Name=dyn_l_foot_folliage_01_MSTR, swing1LimitDegrees=25, swing2LimitDegrees=25, twistLimitDegrees=5}
 {bone1Name=dyn_l_leg_folliage_04, bone2Name=dyn_l_leg_folliage_04_MSTR, swing1LimitDegrees=25, swing2LimitDegrees=25, twistLimitDegrees=5}
 {bone1Name=dyn_l_foot_folliage_02, bone2Name=dyn_l_foot_folliage_01, swing1LimitDegrees=25, swing2LimitDegrees=25, twistLimitDegrees=5}
 {bone1Name=dyn_l_foot_folliage_03, bone2Name=dyn_l_foot_folliage_01, swing1LimitDegrees=25, swing2LimitDegrees=25, twistLimitDegrees=5}
 {bone1Name=dyn_l_leg_folliage_03, bone2Name=dyn_l_leg_folliage_03_MSTR, swing1LimitDegrees=50, swing2LimitDegrees=50, twistLimitDegrees=20}
 {bone1Name=dyn_l_leg_folliage_02, bone2Name=dyn_l_leg_folliage_02_MSTR, swing1LimitDegrees=45, swing2LimitDegrees=45, twistLimitDegrees=20}
 {bone1Name=dyn_l_leg_folliage_01, bone2Name=dyn_l_leg_folliage_01_MSTR, swing1LimitDegrees=50, swing2LimitDegrees=50, twistLimitDegrees=5}
 {bone1Name=dyn_r_foot_folliage_03, bone2Name=dyn_r_foot_folliage_03_MSTR, swing1LimitDegrees=25, swing2LimitDegrees=25, twistLimitDegrees=5}
 {bone1Name=dyn_r_foot_folliage_04, bone2Name=dyn_r_foot_folliage_03, swing1LimitDegrees=25, swing2LimitDegrees=25, twistLimitDegrees=5}
 {bone1Name=dyn_r_foot_folliage_01, bone2Name=dyn_r_foot_folliage_01_MSTR, swing1LimitDegrees=25, swing2LimitDegrees=25, twistLimitDegrees=5}
 {bone1Name=dyn_r_foot_folliage_02, bone2Name=dyn_r_foot_folliage_01, swing1LimitDegrees=25, swing2LimitDegrees=25, twistLimitDegrees=5}
 {bone1Name=dyn_l_butt_folliage_01, bone2Name=dyn_l_butt_folliage_MSTR, swing1LimitDegrees=60, swing2LimitDegrees=60, twistLimitDegrees=30}
 {bone1Name=dyn_l_butt_folliage_02, bone2Name=dyn_l_butt_folliage_01, swing1LimitDegrees=60, swing2LimitDegrees=60, twistLimitDegrees=30}
 {bone1Name=dyn_l_butt_folliage_03, bone2Name=dyn_l_butt_folliage_02, swing1LimitDegrees=60, swing2LimitDegrees=60, twistLimitDegrees=30}
 {bone1Name=dyn_l_torso_folliage_01, bone2Name=dyn_l_torso_folliage_01_MSTR, swing1LimitDegrees=50, swing2LimitDegrees=50, twistLimitDegrees=5}
 {
  bone1Name=dyn_l_scapula_Folliage_Base
  bone2Name=pelvis
  swing1LimitDegrees=45
  swing1Motion=Free
  swing2LimitDegrees=45
  swing2Motion=Free
  twistLimitDegrees=45
  twistMotion=Free
 }
 {
  bone1Name=dyn_l_scapula_Folliage_01
  bone2Name=dyn_l_scapula_Folliage_Base
  swing1LimitDegrees=45
  swing2LimitDegrees=45
  twistLimitDegrees=45
 }
 {bone1Name=dyn_l_scapula_Folliage_02, bone2Name=dyn_l_scapula_Folliage_01, swing1LimitDegrees=45, swing2LimitDegrees=45, twistLimitDegrees=45}
 {bone1Name=dyn_l_scapula_Folliage_03, bone2Name=dyn_l_scapula_Folliage_02, swing1LimitDegrees=45, swing2LimitDegrees=45, twistLimitDegrees=45}
 {
  bone1Name=dyn_r_leg_back_folliage_MSTR
  bone2Name=thigh_twist_01_r
  swing1LimitDegrees=45
  swing1Motion=Free
  swing2LimitDegrees=45
  swing2Motion=Free
  twistLimitDegrees=45
  twistMotion=Free
 }
 {
  bone1Name=dyn_r_leg_back_folliage_01
  bone2Name=dyn_r_leg_back_folliage_MSTR
  swing1LimitDegrees=45
  swing2LimitDegrees=45
  twistLimitDegrees=5
 }
 {
  bone1Name=dyn_r_leg_back_folliage_02
  bone2Name=dyn_r_leg_back_folliage_01
  swing1LimitDegrees=45
  swing2LimitDegrees=45
  twistLimitDegrees=5
 }
 {
  bone1Name=thigh_r
  bone2Name=pelvis
  swing1LimitDegrees=45
  swing1Motion=Free
  swing2LimitDegrees=45
  swing2Motion=Free
  twistLimitDegrees=45
  twistMotion=Free
 }
 {
  bone1Name=thigh_twist_01_r
  bone2Name=thigh_r
  swing1LimitDegrees=45
  swing1Motion=Free
  swing2LimitDegrees=45
  swing2Motion=Free
  twistLimitDegrees=45
  twistMotion=Free
 }
 {
  bone1Name=dyn_r_foot_folliage_05_MSTR
  bone2Name=thigh_r
  swing1LimitDegrees=45
  swing1Motion=Free
  swing2LimitDegrees=45
  swing2Motion=Free
  twistLimitDegrees=45
  twistMotion=Free
 }
 {bone1Name=dyn_r_foot_folliage_05, bone2Name=dyn_r_foot_folliage_05_MSTR, swing1LimitDegrees=45, swing2LimitDegrees=5, twistLimitDegrees=5}
 {
  bone1Name=dyn_l_neck_Folliage_MSTR
  bone2Name=pelvis
  swing1LimitDegrees=45
  swing1Motion=Free
  swing2LimitDegrees=45
  swing2Motion=Free
  twistLimitDegrees=45
  twistMotion=Free
 }
 {bone1Name=dyn_l_neck_Folliage, bone2Name=dyn_l_neck_Folliage_MSTR, swing1LimitDegrees=45, swing2LimitDegrees=45, twistLimitDegrees=45}
 {
  bone1Name=dyn_l_chest_Folliage_MSTR
  bone2Name=pelvis
  swing1LimitDegrees=45
  swing1Motion=Free
  swing2LimitDegrees=45
  swing2Motion=Free
  twistLimitDegrees=45
  twistMotion=Free
 }
 {bone1Name=dyn_l_chest_Folliage, bone2Name=dyn_l_chest_Folliage_MSTR, swing1LimitDegrees=25, swing2LimitDegrees=25, twistLimitDegrees=5}
 {
  bone1Name=dyn_l_shoulder_Folliage_d_MSTR
  bone2Name=pelvis
  swing1LimitDegrees=45
  swing1Motion=Free
  swing2LimitDegrees=45
  swing2Motion=Free
  twistLimitDegrees=45
  twistMotion=Free
 }
 {
  bone1Name=dyn_l_shoulder_Folliage_e_MSTR
  bone2Name=pelvis
  swing1LimitDegrees=45
  swing1Motion=Free
  swing2LimitDegrees=45
  swing2Motion=Free
  twistLimitDegrees=45
  twistMotion=Free
 }
 {
  bone1Name=dyn_l_shoulder_Folliage_d
  bone2Name=dyn_l_shoulder_Folliage_d_MSTR
  swing1LimitDegrees=25
  swing2LimitDegrees=25
  twistLimitDegrees=2
 }
 {
  bone1Name=dyn_l_shoulder_Folliage_e
  bone2Name=dyn_l_shoulder_Folliage_e_MSTR
  swing1LimitDegrees=25
  swing2LimitDegrees=25
  twistLimitDegrees=2
 }
 {
  bone1Name=dyn_r_shoulder_folliage_d_MSTR
  bone2Name=pelvis
  swing1LimitDegrees=45
  swing1Motion=Free
  swing2LimitDegrees=45
  swing2Motion=Free
  twistLimitDegrees=45
  twistMotion=Free
 }
 {
  bone1Name=dyn_r_shoulder_folliage_d
  bone2Name=dyn_r_shoulder_folliage_d_MSTR
  swing1LimitDegrees=25
  swing2LimitDegrees=25
  twistLimitDegrees=5
 }
 {
  bone1Name=dyn_l_arm_folliage_07
  bone2Name=dyn_l_arm_folliage_07_MSTR
  swing1LimitDegrees=45
  swing1Motion=Free
  swing2LimitDegrees=45
  swing2Motion=Free
  twistLimitDegrees=45
 }
 {
  bone1Name=dyn_l_arm_folliage_07_MSTR
  bone2Name=pelvis
  swing1LimitDegrees=45
  swing1Motion=Free
  swing2LimitDegrees=45
  swing2Motion=Free
  twistLimitDegrees=45
  twistMotion=Free
 }
]
