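# Anim Blueprint for the Moss CarbideBlack player body cosmetic, adding secondary motion to foliage and knee pads.
# The copied mesh pose runs through leaf RigidBody sims, blended by front-end and targeting state, using skydiving-driven wind as gravity.
# Weighted constraints, crouch and jump/fall offsets, and spring-based AnimDynamics then adjust those bones.
# The event graph also computes head-swing, backbling, cape and final-gravity values that the AnimGraph never reads.
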AnimBlueprint /BRCosmetics/Characters/Player/Male/Medium/Bodies/M_MED_Moss_CarbideBlack/Meshes/M_MED_Moss_CarbideBlack_AnimBP
target_skeleton=&/M_MED_Moss_CarbideBlack_Skeleton
graphs [
 AnimGraph {nodes [
   n1 = "Transform (Modify) Bone - Bone: dyn_l_kneePad"(
    ComponentPose=Local To Component(LocalPose=Use cached pose 'NotCrouching'())
    Alpha=1
    Translation="0,-2.5,0"
    Rotation="0,0,0"
    Scale="1,1,1"
   ) {
    rotationMode=BMM_Additive
    rotationSpace=BCS_BoneSpace
    scaleMode=BMM_Additive
    scaleSpace=BCS_BoneSpace
    translationMode=BMM_Additive
    translationSpace=BCS_BoneSpace
    alphaBoolBlend {blendOption=Linear}
    boneToModify {boneName=dyn_l_kneePad}
    translation {y=-2.5}
   }
   n2 = "Transform (Modify) Bone - Bone: dyn_r_kneePad"(ComponentPose=@n1, Alpha=1, Translation="0,2.5,0", Rotation="0,0,0", Scale="1,1,1") {
    rotationMode=BMM_Additive
    rotationSpace=BCS_BoneSpace
    scaleMode=BMM_Additive
    scaleSpace=BCS_BoneSpace
    translationMode=BMM_Additive
    translationSpace=BCS_BoneSpace
    alphaBoolBlend {blendOption=Linear}
    boneToModify {boneName=dyn_r_kneePad}
    translation {y=2.5}
   }
   n3 = Component To Local(ComponentPose=@n2)
   n4 = Use cached pose 'NotCrouching'()
   n5 = Get bIsCrouching()
   n6 = "Constraint - dyn_r_foot_folliage_05_MSTR (P)"(
    ComponentPose=Local To Component(LocalPose=Blend Poses by bool(BlendPose_0=@n3, BlendPose_1=@n4, bActiveValue=@n5) {blendType=Linear})
    Alpha=1
    ConstraintWeights_0=1
   ) {
    alphaBoolBlend {blendOption=Linear}
    boneToModify {boneName=dyn_r_foot_folliage_05_MSTR}
    constraintSetup [{targetBone {boneName=foot_r}}]
    constraintWeights [1]
   }
   n7 = "Constraint - dyn_r_leg_back_folliage_MSTR (PPPT)"!(
    ComponentPose=@n6
    Alpha=1
    ConstraintWeights_0=0
    ConstraintWeights_1:float=0.5713
    ConstraintWeights_2:float=0.4287
    ConstraintWeights_3:float=0
   ) -> (Pose) {
    alphaBoolBlend {blendOption=Linear}
    boneToModify {boneName=dyn_r_leg_back_folliage_MSTR}
    constraintSetup [
     {targetBone {boneName=thigh_r}}
     {targetBone {boneName=pelvis}}
     {targetBone {boneName=thigh_twist_01_r}}
     {transformType=Translation, targetBone {boneName=None}}
    ]
    constraintWeights [1,1,1,1]
   }
   n8 = "Constraint - dyn_l_scapula_Folliage_Base (P)"(ComponentPose=@n7, Alpha=0.2, ConstraintWeights_0=1) {
    alphaBoolBlend {blendOption=Linear}
    boneToModify {boneName=dyn_l_scapula_Folliage_Base}
    constraintSetup [{targetBone {boneName=clavicle_l}}]
    constraintWeights [1]
   }
   n9 = "Constraint - dyn_l_shoulder_Folliage_d_MSTR (PPPP)"!(
    ComponentPose=@n8
    Alpha=1
    ConstraintWeights_0=0.0326
    ConstraintWeights_1:float=0.0238
    ConstraintWeights_2:float=0.3802
    ConstraintWeights_3:float=0.5634
   ) -> (Pose) {
    alphaBoolBlend {blendOption=Linear}
    boneToModify {boneName=dyn_l_shoulder_Folliage_d_MSTR}
    constraintSetup [
     {targetBone {boneName=spine_03}}
     {targetBone {boneName=spine_05}}
     {targetBone {boneName=clavicle_l}}
     {targetBone {boneName=upperarm_twist_01_l}}
    ]
    constraintWeights [1,1,1,1]
   }
   n10 = "Constraint - dyn_l_shoulder_Folliage_e_MSTR (PPPPPP)"!(
    ComponentPose=@n9
    Alpha=1
    ConstraintWeights_0=0.2303
    ConstraintWeights_1:float=0.0009
    ConstraintWeights_2:float=0.0239
    ConstraintWeights_3:float=0.0195
    ConstraintWeights_4:float=0.724
    ConstraintWeights_5:float=0.0013
   ) -> (Pose) {
    alphaBoolBlend {blendOption=Linear}
    boneToModify {boneName=dyn_l_shoulder_Folliage_e_MSTR}
    constraintSetup [
     {targetBone {boneName=clavicle_l}}
     {targetBone {boneName=upperarm_twist_02_l}}
     {targetBone {boneName=spine_03}}
     {targetBone {boneName=spine_05}}
     {targetBone {boneName=upperarm_twist_01_l}}
     {targetBone {boneName=upperarm_l}}
    ]
    constraintWeights [1,1,1,1,1,1]
   }
   n11 = "Constraint - dyn_l_shoulder_Folliage_f_MSTR (PPPPPP)"!(
    ComponentPose=@n10
    Alpha=1
    ConstraintWeights_0=0.3159
    ConstraintWeights_1:float=0.0005
    ConstraintWeights_2:float=0.0291
    ConstraintWeights_3:float=0.0293
    ConstraintWeights_4:float=0.6245
    ConstraintWeights_5:float=0.0007
   ) -> (Pose) {
    alphaBoolBlend {blendOption=Linear}
    boneToModify {boneName=dyn_l_shoulder_Folliage_f_MSTR}
    constraintSetup [
     {targetBone {boneName=clavicle_l}}
     {targetBone {boneName=upperarm_twist_02_l}}
     {targetBone {boneName=spine_03}}
     {targetBone {boneName=spine_05}}
     {targetBone {boneName=upperarm_twist_01_l}}
     {targetBone {boneName=upperarm_l}}
    ]
    constraintWeights [1,1,1,1,1,1]
   }
   n12 = "Constraint - dyn_r_shoulder_folliage_d_MSTR (PPPPPPP)"!(
    ComponentPose=@n11
    Alpha=1
    ConstraintWeights_0=0.0001
    ConstraintWeights_1:float=0.005
    ConstraintWeights_2:float=0.0002
    ConstraintWeights_3:float=0.9877
    ConstraintWeights_4:float=0.0066
    ConstraintWeights_5:float=0
    ConstraintWeights_6:float=0.0003
   ) -> (Pose) {
    alphaBoolBlend {blendOption=Linear}
    boneToModify {boneName=dyn_r_shoulder_folliage_d_MSTR}
    constraintSetup [
     {targetBone {boneName=lowerarm_twist_02_r}}
     {targetBone {boneName=upperarm_r}}
     {targetBone {boneName=lowerarm_r}}
     {targetBone {boneName=upperarm_twist_01_r}}
     {targetBone {boneName=upperarm_twist_02_r}}
     {targetBone {boneName=spine_03}}
     {targetBone {boneName=clavicle_r}}
    ]
    constraintWeights [1,1,1,1,1,1,1]
   }
   n13 = "Constraint - dyn_r_arm_folliage_07_MSTR (PPPPPPPP)"!(
    ComponentPose=@n12
    Alpha=1
    ConstraintWeights_0=1
    ConstraintWeights_1:float=1
    ConstraintWeights_2:float=1
    ConstraintWeights_3:float=1
    ConstraintWeights_4:float=1
    ConstraintWeights_5:float=1
    ConstraintWeights_6:float=1
    ConstraintWeights_7:float=1
   ) -> (Pose) {
    alphaBoolBlend {blendOption=Linear}
    boneToModify {boneName=dyn_r_arm_folliage_07_MSTR}
    constraintSetup [
     {targetBone {boneName=lowerarm_twist_02_r}}
     {targetBone {boneName=upperarm_r}}
     {targetBone {boneName=lowerarm_r}}
     {targetBone {boneName=lowerarm_twist_01_r}}
     {targetBone {boneName=hand_r}}
     {targetBone {boneName=upperarm_twist_02_r}}
     {targetBone {boneName=dyn_r_arm_folliage_07_MSTR}}
     {targetBone {boneName=thumb_01_r}}
    ]
    constraintWeights [1,1,1,1,1,1,1,1]
   }
   n14 = "Constraint - dyn_r_arm_folliage_06_MSTR (PPPPPPPP)"!(
    ComponentPose=@n13
    Alpha=1
    ConstraintWeights_0=1
    ConstraintWeights_1:float=1
    ConstraintWeights_2:float=1
    ConstraintWeights_3:float=1
    ConstraintWeights_4:float=1
    ConstraintWeights_5:float=1
    ConstraintWeights_6:float=1
    ConstraintWeights_7:float=1
   ) -> (Pose) {
    alphaBoolBlend {blendOption=Linear}
    boneToModify {boneName=dyn_r_arm_folliage_06_MSTR}
    constraintSetup [
     {targetBone {boneName=lowerarm_twist_02_r}}
     {targetBone {boneName=upperarm_r}}
     {targetBone {boneName=lowerarm_r}}
     {targetBone {boneName=lowerarm_twist_01_r}}
     {targetBone {boneName=hand_r}}
     {targetBone {boneName=upperarm_twist_02_r}}
     {targetBone {boneName=dyn_r_arm_folliage_07_MSTR}}
     {targetBone {boneName=thumb_01_r}}
    ]
    constraintWeights [1,1,1,1,1,1,1,1]
   }
   n15 = "Constraint - dyn_l_neck_Folliage_MSTR (PPPPPPPP)"!(
    ComponentPose=@n14
    Alpha=1
    ConstraintWeights_0=0.0307
    ConstraintWeights_1:float=0.0789
    ConstraintWeights_2:float=0.0917
    ConstraintWeights_3:float=0.048
    ConstraintWeights_4:float=0.6499
    ConstraintWeights_5:float=0.0046
    ConstraintWeights_6:float=0.0695
    ConstraintWeights_7:float=0.0267
   ) -> (Pose) {
    alphaBoolBlend {blendOption=Linear}
    boneToModify {boneName=dyn_l_neck_Folliage_MSTR}
    constraintSetup [
     {targetBone {boneName=dyn_backNeck}}
     {targetBone {boneName=clavicle_l}}
     {targetBone {boneName=spine_03}}
     {targetBone {boneName=spine_04}}
     {targetBone {boneName=spine_05}}
     {targetBone {boneName=clavicle_r}}
     {targetBone {boneName=neck_02}}
     {targetBone {boneName=neck_01}}
    ]
    constraintWeights [1,1,1,1,1,1,1,1]
   }
   n16 = "Transform (Modify) Bone - Bone: dyn_l_torso_folliage_01_MSTR"(ComponentPose=@n15, Alpha="To Float (Boolean)"(InBool=Get bIsCrouching()), Translation="3,-3,0", Rotation="0,-20,0", Scale="1,1,1") {
    rotationMode=BMM_Additive
    scaleMode=BMM_Additive
    translationMode=BMM_Additive
    alphaBoolBlend {blendOption=Linear}
    boneToModify {boneName=dyn_l_torso_folliage_01_MSTR}
    rotation {yaw=-20}
    translation=<3 -3>
   }
   n17 = Anim Dynamics - Bone: dyn_l_leg_folliage_02(ComponentPose=@n16, Alpha=0.6, ExternalForce="0,0,0") {
    angularDampingOverride=0.9999
    angularSpringConstant=50
    bAngularSpring=true
    bGravityOverrideInSimSpace=true
    bOverrideAngularDamping=true
    simulationSpace=RootRelative
    alphaBoolBlend {blendOption=Linear}
    boundBone {boneName=dyn_l_leg_folliage_02}
    componentLinearAccScale=<1 1 1>
    componentLinearVelScale=<1 1 1>
    physicsBodyDefinitions [{
      sphereCollisionRadius=0
      boundBone {boneName=dyn_l_leg_folliage_02}
      boxExtents=<20 20 10>
      constraintSetup {angularLimitsMax=<45 45 45>, angularLimitsMin=<-45 -45 -45>, angularTarget=<0 0 0>}
      localJointOffset=<0 0 -5>
     }]
   }
   n18 = Anim Dynamics - Bone: dyn_l_leg_folliage_03(ComponentPose=@n17, Alpha=0.8, ExternalForce="0,0,0") {
    angularDampingOverride=0.9999
    angularSpringConstant=50
    bAngularSpring=true
    bGravityOverrideInSimSpace=true
    bOverrideAngularDamping=true
    simulationSpace=RootRelative
    alphaBoolBlend {blendOption=Linear}
    boundBone {boneName=dyn_l_leg_folliage_03}
    componentLinearAccScale=<1 1 1>
    componentLinearVelScale=<1 1 1>
    physicsBodyDefinitions [{
      sphereCollisionRadius=0
      boundBone {boneName=dyn_l_leg_folliage_03}
      boxExtents=<20 20 10>
      constraintSetup {angularLimitsMax=<45 45 45>, angularLimitsMin=<-45 -45 -45>, angularTarget=<0 0 0>}
      localJointOffset=<0 0 -5>
     }]
   }
   n19 = Anim Dynamics - Bone: dyn_l_leg_folliage_04(ComponentPose=@n18, Alpha=0.7, ExternalForce="0,0,0") {
    angularDampingOverride=0.9999
    angularSpringConstant=50
    bAngularSpring=true
    bGravityOverrideInSimSpace=true
    bOverrideAngularDamping=true
    simulationSpace=RootRelative
    alphaBoolBlend {blendOption=Linear}
    boundBone {boneName=dyn_l_leg_folliage_04}
    componentLinearAccScale=<1 1 1>
    componentLinearVelScale=<1 1 1>
    physicsBodyDefinitions [{
      sphereCollisionRadius=0
      boundBone {boneName=dyn_l_leg_folliage_04}
      boxExtents=<20 20 10>
      constraintSetup {angularLimitsMax=<45 45 45>, angularLimitsMin=<-45 -45 -45>, angularTarget=<0 0 0>}
      localJointOffset=<0 0 -5>
     }]
   }
   n20 = Anim Dynamics - Bone: dyn_l_leg_folliage_05(ComponentPose=@n19, Alpha=1, ExternalForce="0,0,0") {
    angularDampingOverride=0.9999
    angularSpringConstant=50
    bAngularSpring=true
    bGravityOverrideInSimSpace=true
    bOverrideAngularDamping=true
    simulationSpace=RootRelative
    alphaBoolBlend {blendOption=Linear}
    boundBone {boneName=dyn_l_leg_folliage_05}
    componentLinearAccScale=<1 1 1>
    componentLinearVelScale=<1 1 1>
    physicsBodyDefinitions [{
      sphereCollisionRadius=0
      boundBone {boneName=dyn_l_leg_folliage_05}
      boxExtents=<20 20 10>
      constraintSetup {angularLimitsMax=<45 45 45>, angularLimitsMin=<-45 -45 -45>, angularTarget=<0 0 0>}
      localJointOffset=<0 0 -5>
     }]
   }
   n21 = Anim Dynamics - Bone: dyn_l_leg_folliage_06(ComponentPose=@n20, Alpha=0.8, ExternalForce="0,0,0") {
    angularDampingOverride=0.9999
    angularSpringConstant=50
    bAngularSpring=true
    bGravityOverrideInSimSpace=true
    bOverrideAngularDamping=true
    simulationSpace=RootRelative
    alphaBoolBlend {blendOption=Linear}
    boundBone {boneName=dyn_l_leg_folliage_06}
    componentLinearAccScale=<1 1 1>
    componentLinearVelScale=<1 1 1>
    physicsBodyDefinitions [{
      sphereCollisionRadius=0
      boundBone {boneName=dyn_l_leg_folliage_06}
      boxExtents=<20 20 10>
      constraintSetup {angularLimitsMax=<45 45 45>, angularLimitsMin=<-45 -45 -45>, angularTarget=<0 0 0>}
      localJointOffset=<0 0 -5>
     }]
   }
   n22 = Anim Dynamics - Bone: dyn_l_leg_folliage_07(ComponentPose=@n21, Alpha=0.5, ExternalForce="0,0,0") {
    angularDampingOverride=0.9999
    angularSpringConstant=50
    bAngularSpring=true
    bGravityOverrideInSimSpace=true
    bOverrideAngularDamping=true
    simulationSpace=RootRelative
    alphaBoolBlend {blendOption=Linear}
    boundBone {boneName=dyn_l_leg_folliage_07}
    componentLinearAccScale=<1 1 1>
    componentLinearVelScale=<1 1 1>
    physicsBodyDefinitions [{
      sphereCollisionRadius=0
      boundBone {boneName=dyn_l_leg_folliage_07}
      boxExtents=<20 20 10>
      constraintSetup {angularLimitsMax=<45 45 45>, angularLimitsMin=<-45 -45 -45>, angularTarget=<0 0 0>}
      localJointOffset=<0 0 -5>
     }]
   }
   n23 = Anim Dynamics - Bone: dyn_l_torso_folliage_01(ComponentPose=@n22, Alpha=1, ExternalForce="0,0,0") {
    angularDampingOverride=0.9999
    angularSpringConstant=50
    bAngularSpring=true
    bGravityOverrideInSimSpace=true
    bOverrideAngularDamping=true
    simulationSpace=RootRelative
    alphaBoolBlend {blendOption=Linear}
    boundBone {boneName=dyn_l_torso_folliage_01}
    componentLinearAccScale=<1 1 1>
    componentLinearVelScale=<1 1 1>
    physicsBodyDefinitions [{
      sphereCollisionRadius=0
      boundBone {boneName=dyn_l_torso_folliage_01}
      boxExtents=<20 20 10>
      constraintSetup {angularLimitsMax=<45 45 45>, angularLimitsMin=<-45 -45 -45>, angularTarget=<0 0 0>}
      localJointOffset=<0 0 -5>
     }]
   }
   n24 = Anim Dynamics - Bone: dyn_l_torso_folliage_02(ComponentPose=@n23, Alpha=1, ExternalForce="0,0,0") {
    angularDampingOverride=0.9999
    angularSpringConstant=50
    bAngularSpring=true
    bGravityOverrideInSimSpace=true
    bOverrideAngularDamping=true
    simulationSpace=RootRelative
    alphaBoolBlend {blendOption=Linear}
    boundBone {boneName=dyn_l_torso_folliage_02}
    componentLinearAccScale=<1 1 1>
    componentLinearVelScale=<1 1 1>
    physicsBodyDefinitions [{
      sphereCollisionRadius=0
      boundBone {boneName=dyn_l_torso_folliage_02}
      boxExtents=<20 20 10>
      constraintSetup {angularLimitsMax=<45 45 45>, angularLimitsMin=<-45 -45 -45>, angularTarget=<0 0 0>}
      localJointOffset=<0 0 -5>
     }]
   }
   n25 = Anim Dynamics - Bone: dyn_l_torso_folliage_03(ComponentPose=@n24, Alpha=1, ExternalForce="0,0,0") {
    angularDampingOverride=0.9999
    angularSpringConstant=50
    bAngularSpring=true
    bGravityOverrideInSimSpace=true
    bOverrideAngularDamping=true
    simulationSpace=RootRelative
    alphaBoolBlend {blendOption=Linear}
    boundBone {boneName=dyn_l_torso_folliage_03}
    componentLinearAccScale=<1 1 1>
    componentLinearVelScale=<1 1 1>
    physicsBodyDefinitions [{
      sphereCollisionRadius=0
      boundBone {boneName=dyn_l_torso_folliage_03}
      boxExtents=<20 20 10>
      constraintSetup {angularLimitsMax=<45 45 45>, angularLimitsMin=<-45 -45 -45>, angularTarget=<0 0 0>}
      localJointOffset=<0 0 -5>
     }]
   }
   n26 = Anim Dynamics - Bone: dyn_l_torso_folliage_04(ComponentPose=@n25, Alpha=1, ExternalForce="0,0,0") {
    angularDampingOverride=0.9999
    angularSpringConstant=50
    bAngularSpring=true
    bGravityOverrideInSimSpace=true
    bOverrideAngularDamping=true
    simulationSpace=RootRelative
    alphaBoolBlend {blendOption=Linear}
    boundBone {boneName=dyn_l_torso_folliage_04}
    componentLinearAccScale=<1 1 1>
    componentLinearVelScale=<1 1 1>
    physicsBodyDefinitions [{
      sphereCollisionRadius=0
      boundBone {boneName=dyn_l_torso_folliage_04}
      boxExtents=<20 20 10>
      constraintSetup {angularLimitsMax=<45 45 45>, angularLimitsMin=<-45 -45 -45>, angularTarget=<0 0 0>}
      localJointOffset=<0 0 -5>
     }]
   }
   n27 = Anim Dynamics - Bone: dyn_l_torso_folliage_05(ComponentPose=@n26, Alpha=1, ExternalForce="0,0,0") {
    angularDampingOverride=0.9999
    angularSpringConstant=50
    bAngularSpring=true
    bGravityOverrideInSimSpace=true
    bOverrideAngularDamping=true
    simulationSpace=RootRelative
    alphaBoolBlend {blendOption=Linear}
    boundBone {boneName=dyn_l_torso_folliage_05}
    componentLinearAccScale=<1 1 1>
    componentLinearVelScale=<1 1 1>
    physicsBodyDefinitions [{
      sphereCollisionRadius=0
      boundBone {boneName=dyn_l_torso_folliage_05}
      boxExtents=<20 20 10>
      constraintSetup {angularLimitsMax=<45 45 45>, angularLimitsMin=<-45 -45 -45>, angularTarget=<0 0 0>}
      localJointOffset=<0 0 -5>
     }]
   }
   n28 = Anim Dynamics - Bone: dyn_l_torso_folliage_06(ComponentPose=@n27, Alpha=1, ExternalForce="0,0,0") {
    angularDampingOverride=0.9999
    angularSpringConstant=50
    bAngularSpring=true
    bGravityOverrideInSimSpace=true
    bOverrideAngularDamping=true
    simulationSpace=RootRelative
    alphaBoolBlend {blendOption=Linear}
    boundBone {boneName=dyn_l_torso_folliage_06}
    componentLinearAccScale=<1 1 1>
    componentLinearVelScale=<1 1 1>
    physicsBodyDefinitions [{
      sphereCollisionRadius=0
      boundBone {boneName=dyn_l_torso_folliage_06}
      boxExtents=<20 20 10>
      constraintSetup {angularLimitsMax=<45 45 45>, angularLimitsMin=<-45 -45 -45>, angularTarget=<0 0 0>}
      localJointOffset=<0 0 -5>
     }]
   }
   n29 = Anim Dynamics - Bone: dyn_l_arm_folliage_01(ComponentPose=@n28, Alpha=1, ExternalForce="0,0,0") {
    angularDampingOverride=0.9999
    angularSpringConstant=50
    bAngularSpring=true
    bGravityOverrideInSimSpace=true
    bOverrideAngularDamping=true
    simulationSpace=RootRelative
    alphaBoolBlend {blendOption=Linear}
    boundBone {boneName=dyn_l_arm_folliage_01}
    componentLinearAccScale=<1 1 1>
    componentLinearVelScale=<1 1 1>
    physicsBodyDefinitions [{
      sphereCollisionRadius=0
      boundBone {boneName=dyn_l_arm_folliage_01}
      boxExtents=<20 20 10>
      constraintSetup {angularLimitsMax=<45 45 45>, angularLimitsMin=<-45 -45 -45>, angularTarget=<0 0 0>}
      localJointOffset=<0 0 -5>
     }]
   }
   n30 = Anim Dynamics - Bone: dyn_l_arm_folliage_02(ComponentPose=@n29, Alpha=1, ExternalForce="0,0,0") {
    angularDampingOverride=0.9999
    angularSpringConstant=50
    bAngularSpring=true
    bGravityOverrideInSimSpace=true
    bOverrideAngularDamping=true
    simulationSpace=RootRelative
    alphaBoolBlend {blendOption=Linear}
    boundBone {boneName=dyn_l_arm_folliage_02}
    componentLinearAccScale=<1 1 1>
    componentLinearVelScale=<1 1 1>
    physicsBodyDefinitions [{
      sphereCollisionRadius=0
      boundBone {boneName=dyn_l_arm_folliage_02}
      boxExtents=<20 20 10>
      constraintSetup {angularLimitsMax=<45 45 45>, angularLimitsMin=<-45 -45 -45>, angularTarget=<0 0 0>}
      localJointOffset=<0 0 -5>
     }]
   }
   n31 = Anim Dynamics - Bone: dyn_l_arm_folliage_03(ComponentPose=@n30, Alpha=1, ExternalForce="0,0,0") {
    angularDampingOverride=0.9999
    angularSpringConstant=50
    bAngularSpring=true
    bGravityOverrideInSimSpace=true
    bOverrideAngularDamping=true
    simulationSpace=RootRelative
    alphaBoolBlend {blendOption=Linear}
    boundBone {boneName=dyn_l_arm_folliage_03}
    componentLinearAccScale=<1 1 1>
    componentLinearVelScale=<1 1 1>
    physicsBodyDefinitions [{
      sphereCollisionRadius=0
      boundBone {boneName=dyn_l_arm_folliage_03}
      boxExtents=<20 20 10>
      constraintSetup {angularLimitsMax=<45 45 45>, angularLimitsMin=<-45 -45 -45>, angularTarget=<0 0 0>}
      localJointOffset=<0 0 -5>
     }]
   }
   n32 = Anim Dynamics - Bone: dyn_l_arm_folliage_04(ComponentPose=@n31, Alpha=1, ExternalForce="0,0,0") {
    angularDampingOverride=0.9999
    angularSpringConstant=50
    bAngularSpring=true
    bGravityOverrideInSimSpace=true
    bOverrideAngularDamping=true
    simulationSpace=RootRelative
    alphaBoolBlend {blendOption=Linear}
    boundBone {boneName=dyn_l_arm_folliage_04}
    componentLinearAccScale=<1 1 1>
    componentLinearVelScale=<1 1 1>
    physicsBodyDefinitions [{
      sphereCollisionRadius=0
      boundBone {boneName=dyn_l_arm_folliage_04}
      boxExtents=<20 20 10>
      constraintSetup {angularLimitsMax=<45 45 45>, angularLimitsMin=<-45 -45 -45>, angularTarget=<0 0 0>}
      localJointOffset=<0 0 -5>
     }]
   }
   n33 = Anim Dynamics - Bone: dyn_l_arm_folliage_05(ComponentPose=@n32, Alpha=1, ExternalForce="0,0,0") {
    angularDampingOverride=0.9999
    angularSpringConstant=50
    bAngularSpring=true
    bGravityOverrideInSimSpace=true
    bOverrideAngularDamping=true
    simulationSpace=RootRelative
    alphaBoolBlend {blendOption=Linear}
    boundBone {boneName=dyn_l_arm_folliage_05}
    componentLinearAccScale=<1 1 1>
    componentLinearVelScale=<1 1 1>
    physicsBodyDefinitions [{
      sphereCollisionRadius=0
      boundBone {boneName=dyn_l_arm_folliage_05}
      boxExtents=<20 20 10>
      constraintSetup {angularLimitsMax=<45 45 45>, angularLimitsMin=<-45 -45 -45>, angularTarget=<0 0 0>}
      localJointOffset=<0 0 -5>
     }]
   }
   n34 = Anim Dynamics - Bone: dyn_l_arm_folliage_06(ComponentPose=@n33, Alpha=1, ExternalForce="0,0,0") {
    angularDampingOverride=0.9999
    angularSpringConstant=50
    bAngularSpring=true
    bGravityOverrideInSimSpace=true
    bOverrideAngularDamping=true
    simulationSpace=RootRelative
    alphaBoolBlend {blendOption=Linear}
    boundBone {boneName=dyn_l_arm_folliage_06}
    componentLinearAccScale=<1 1 1>
    componentLinearVelScale=<1 1 1>
    physicsBodyDefinitions [{
      sphereCollisionRadius=0
      boundBone {boneName=dyn_l_arm_folliage_06}
      boxExtents=<20 20 10>
      constraintSetup {angularLimitsMax=<45 45 45>, angularLimitsMin=<-45 -45 -45>, angularTarget=<0 0 0>}
      localJointOffset=<0 0 -5>
     }]
   }
   n35 = Anim Dynamics - Bone: dyn_l_arm_folliage_07(ComponentPose=@n34, Alpha=1, ExternalForce="0,0,0") {
    angularDampingOverride=0.9999
    angularSpringConstant=50
    bAngularSpring=true
    bGravityOverrideInSimSpace=true
    bOverrideAngularDamping=true
    simulationSpace=RootRelative
    alphaBoolBlend {blendOption=Linear}
    boundBone {boneName=dyn_l_arm_folliage_07}
    componentLinearAccScale=<1 1 1>
    componentLinearVelScale=<1 1 1>
    physicsBodyDefinitions [{
      sphereCollisionRadius=0
      boundBone {boneName=dyn_l_arm_folliage_07}
      boxExtents=<20 20 10>
      constraintSetup {angularLimitsMax=<45 45 45>, angularLimitsMin=<-45 -45 -45>, angularTarget=<0 0 0>}
      localJointOffset=<0 0 -5>
     }]
   }
   n36 = Anim Dynamics - Bone: dyn_r_arm_folliage_03(ComponentPose=@n35, Alpha=1, ExternalForce="0,0,0") {
    angularDampingOverride=0.9999
    angularSpringConstant=50
    bAngularSpring=true
    bGravityOverrideInSimSpace=true
    bOverrideAngularDamping=true
    simulationSpace=RootRelative
    alphaBoolBlend {blendOption=Linear}
    boundBone {boneName=dyn_r_arm_folliage_03}
    componentLinearAccScale=<1 1 1>
    componentLinearVelScale=<1 1 1>
    physicsBodyDefinitions [{
      sphereCollisionRadius=0
      boundBone {boneName=dyn_r_arm_folliage_03}
      boxExtents=<20 20 10>
      constraintSetup {angularLimitsMax=<45 45 45>, angularLimitsMin=<-45 -45 -45>, angularTarget=<0 0 0>}
      localJointOffset=<0 0 -5>
     }]
   }
   n37 = Anim Dynamics - Bone: dyn_r_arm_folliage_06(ComponentPose=@n36, Alpha=1, ExternalForce="0,0,0") {
    angularDampingOverride=0.9999
    angularSpringConstant=50
    bAngularSpring=true
    bGravityOverrideInSimSpace=true
    bOverrideAngularDamping=true
    simulationSpace=RootRelative
    alphaBoolBlend {blendOption=Linear}
    boundBone {boneName=dyn_r_arm_folliage_06}
    componentLinearAccScale=<1 1 1>
    componentLinearVelScale=<1 1 1>
    physicsBodyDefinitions [{
      sphereCollisionRadius=0
      boundBone {boneName=dyn_r_arm_folliage_06}
      boxExtents=<20 20 10>
      constraintSetup {angularLimitsMax=<45 45 45>, angularLimitsMin=<-45 -45 -45>, angularTarget=<0 0 0>}
      localJointOffset=<0 0 -5>
     }]
   }
   n38 = Anim Dynamics - Bone: dyn_r_arm_folliage_07(ComponentPose=@n37, Alpha=1, ExternalForce="0,0,0") {
    angularDampingOverride=0.9999
    angularSpringConstant=50
    bAngularSpring=true
    bGravityOverrideInSimSpace=true
    bOverrideAngularDamping=true
    simulationSpace=RootRelative
    alphaBoolBlend {blendOption=Linear}
    boundBone {boneName=dyn_r_arm_folliage_07}
    componentLinearAccScale=<1 1 1>
    componentLinearVelScale=<1 1 1>
    physicsBodyDefinitions [{
      sphereCollisionRadius=0
      boundBone {boneName=dyn_r_arm_folliage_07}
      boxExtents=<20 20 10>
      constraintSetup {angularLimitsMax=<45 45 45>, angularLimitsMin=<-45 -45 -45>, angularTarget=<0 0 0>}
      localJointOffset=<0 0 -5>
     }]
   }
   n39 = "Transform (Modify) Bone - Bone: dyn_l_kneePad"(ComponentPose=@n38, Alpha=1, Translation="0,-2.5,0", Rotation="0,0,0", Scale="1,1,1") {
    rotationMode=BMM_Additive
    rotationSpace=BCS_BoneSpace
    scaleMode=BMM_Additive
    scaleSpace=BCS_BoneSpace
    translationMode=BMM_Additive
    translationSpace=BCS_BoneSpace
    alphaBoolBlend {blendOption=Linear}
    boneToModify {boneName=dyn_l_kneePad}
    translation {y=-2.5}
   }
   n40 = "Transform (Modify) Bone - Bone: dyn_r_kneePad"(ComponentPose=@n39, Alpha=1, Translation="0,2.5,0", Rotation="0,0,0", Scale="1,1,1") {
    rotationMode=BMM_Additive
    rotationSpace=BCS_BoneSpace
    scaleMode=BMM_Additive
    scaleSpace=BCS_BoneSpace
    translationMode=BMM_Additive
    translationSpace=BCS_BoneSpace
    alphaBoolBlend {blendOption=Linear}
    boneToModify {boneName=dyn_r_kneePad}
    translation {y=2.5}
   }
   n41 = Get Jump_Falling() -> (Jump_Falling:bool)
   n42 = "Transform (Modify) Bone - Bone: dyn_l_kneePad"!(ComponentPose=@n40, bAlphaBoolEnabled:bool=@n41, Translation="0,-3.5,0", Rotation="0,0,0", Scale="1,1,1") -> (Pose) {
    alphaInputType=Bool
    rotationMode=BMM_Additive
    rotationSpace=BCS_BoneSpace
    scaleMode=BMM_Additive
    scaleSpace=BCS_BoneSpace
    translationMode=BMM_Additive
    translationSpace=BCS_BoneSpace
    alphaBoolBlend {blendOption=Linear}
    boneToModify {boneName=dyn_l_kneePad}
    translation {y=-3.5}
   }
   n43 = "Transform (Modify) Bone - Bone: dyn_r_kneePad"!(ComponentPose=@n42, bAlphaBoolEnabled:bool=@n41, Translation="0,3.5,0", Rotation="0,0,0", Scale="1,1,1") -> (Pose) {
    alphaInputType=Bool
    rotationMode=BMM_Additive
    rotationSpace=BCS_BoneSpace
    scaleMode=BMM_Additive
    scaleSpace=BCS_BoneSpace
    translationMode=BMM_Additive
    translationSpace=BCS_BoneSpace
    alphaBoolBlend {blendOption=Linear}
    boneToModify {boneName=dyn_r_kneePad}
    translation {y=3.5}
   }
   Output Pose(Result=Component To Local(ComponentPose=@n43)) {layerGroup=None}
   n44 = RigidBody!(
    ComponentPose=Local To Component(LocalPose=Copy Pose From Mesh() {bUseAttachedParent=true})
    Alpha
    OverrideWorldGravity:Vector=Get windVector() -> (windVector:Vector)
    ExternalForce
   ) -> (Pose) {
    bClampLinearTranslationLimitToRefPose=true
    bOverrideWorldGravity=true
    overridePhysicsAsset=&/M_MED_Moss_CarbideBlack_Leaf_Physics
    simulationSpace=BaseBoneSpace
    alphaBoolBlend {blendOption=Linear}
    baseBoneRef {boneName=root}
    componentAppliedLinearAccClamp=<5000 5000 5000>
    componentLinearAccScale=<0.4 0.4 0.5>
    componentLinearVelScale=<0.7 0.7 0.7>
    overrideWorldGravity {z=-980}
   }
   n45 = RigidBody(
    ComponentPose=Local To Component(LocalPose=Copy Pose From Mesh() {bUseAttachedParent=true})
    Alpha=0.7
    OverrideWorldGravity=Get windVector()
   ) {
    bClampLinearTranslationLimitToRefPose=true
    bOverrideWorldGravity=true
    overridePhysicsAsset=&M_MED_Moss_CarbideBlack_Leaf_Physics
    simulationSpace=BaseBoneSpace
    alphaBoolBlend {blendOption=Linear}
    baseBoneRef {boneName=root}
    componentAppliedLinearAccClamp=<5000 5000 5000>
    componentLinearAccScale=<0.4 0.4 0.5>
    componentLinearVelScale=<0.7 0.7 0.7>
    overrideWorldGravity {z=-980}
   }
   n46 = Blend Poses by bool(BlendPose_0=Component To Local(ComponentPose=@n45), BlendPose_1=Component To Local(ComponentPose=@n44), bActiveValue=Get bIsTargeting()) {blendType=Linear}
   n47 = RigidBody(
    ComponentPose=Local To Component(LocalPose=Copy Pose From Mesh() {bUseAttachedParent=true})
    Alpha=0.6
    OverrideWorldGravity=Get windVector()
   ) {
    bClampLinearTranslationLimitToRefPose=true
    bOverrideWorldGravity=true
    overridePhysicsAsset=&M_MED_Moss_CarbideBlack_Leaf_Physics
    simulationSpace=BaseBoneSpace
    alphaBoolBlend {blendOption=Linear}
    baseBoneRef {boneName=root}
    componentAppliedLinearAccClamp=<5000 5000 5000>
    componentLinearAccScale=<0.4 0.4 0.5>
    componentLinearVelScale=<0.7 0.7 0.7>
    overrideWorldGravity {z=-980}
   }
   n48 = Component To Local(ComponentPose=@n47)
   n49 = Get bIsInFrontEnd()
   n50 = "Constraint - dyn_l_kneePad (PP)"!(
    ComponentPose=Local To Component(LocalPose=Blend Poses by bool(BlendPose_0=@n48, BlendPose_1=@n46, bActiveValue=@n49) {blendType=Linear})
    Alpha=1
    ConstraintWeights_0=0.5
    ConstraintWeights_1:float=0.5
   ) -> (Pose) {
    alphaBoolBlend {blendOption=Linear}
    boneToModify {boneName=dyn_l_kneePad}
    constraintSetup [{targetBone {boneName=thigh_twist_01_l}}, {targetBone {boneName=calf_l}}]
    constraintWeights [1,1]
   }
   n51 = "Constraint - dyn_r_kneePad (PP)"!(ComponentPose=@n50, Alpha=1, ConstraintWeights_0=0.5, ConstraintWeights_1:float=0.5) -> (Pose) {
    alphaBoolBlend {blendOption=Linear}
    boneToModify {boneName=dyn_r_kneePad}
    constraintSetup [{targetBone {boneName=thigh_twist_01_r}}, {targetBone {boneName=calf_r}}]
    constraintWeights [1,1]
   }
   Save cached pose 'NotCrouching'(Pose=Component To Local(ComponentPose=@n51)) {cacheName=NotCrouching}
  ]}
 EventGraph {nodes [
   Event BlueprintUpdateAnimation()
   > n1 = Sequence!(execute) -> (then_0, then_1, then_2:Exec, then_3:Exec)
   n2 = Get bIsTargeting()
   n3 = BreakTransform(InTransform=GetDeltaTransformFromRefPose(self=GetOwnerSkeletalMeshComponent(), BoneName=head, BaseName=spine_05))
   > Set headSwingAlpha(headSwingAlpha:double=MapRangeClamped(Value=BreakRotator(InRot=@n3.Rotation).Yaw, InRangeA=0, InRangeB=-45, OutRangeA="To Float (Boolean)"(InBool=@n2))) -> (Output_Get:double)
   n4 = Break Fort Anim Input Skydiving External Force(FortAnimInput_SkydivingExternalForce=Get SkydivingExternalForceInput()) -> (
    bUseSkydivingVectorForce=~
    SkydivingVectorForce=~
    HeadToPelvisDirection=~
    FloatingMultiplier=~
    FloatingAdditive=~
    DivingMultiplier=~
    DivingAdditive=~
    ParachutingMultiplier=~
    ParachutingAdditive=~
   )
   n5 = Set windVector(
    execute=@n1.then_1
    windVector:Vector=SelectVector(A=vector * float(A=@n4.SkydivingVectorForce, B=0.15), B="0, 0,-980", bPickA=Get bIsSkyDiving())
   ) -> (Output_Get:Vector)
   > Set windAmbientVector(windAmbientVector:Vector=vector * float(A=@n4.SkydivingVectorForce, B=10)) -> (Output_Get:Vector)
   n6 = GetOwnerSkeletalMeshComponent()
   n7 = BreakRotator(InRot=BreakTransform(InTransform=GetDeltaTransformFromRefPose(self=@n6, BoneName=pelvis, BaseName=root)).Rotation)
   n8 = SelectFloat(A=980, B=-980, bPickA="float > float"(A="Absolute (Float)"(A=@n7.Yaw), B=90) {is=title K2Node_CallFunction})
   > Set finalGravity(finalGravity:Vector=SelectVector(A=MakeVector(Y=0, Z=@n8), B=@n5, bPickA=Get bIsPlayingEmote())) -> (Output_Get:Vector)
   Set Jump_Falling(execute=@n1.then_2, Jump_Falling:bool=OR Boolean(A=Get bIsJumping(), B=Get bIsFalling())) -> (Output_Get:bool)
   Event OnCharacterCustomizationLoaded()
   > Do Once()
   n9 = CheckForBackbling()
   > Set isWearingBackblingAlpha(isWearingBackblingAlpha:double="To Float (Boolean)"(InBool=OR Boolean(A=@n9.bIsBackblingLong, B=@n9))) -> (Output_Get:double)
   > Branch(Condition=@n9.bIsBackblingLong)
   n10 = GetPartSkeletalMeshComponent()
   n11 = FindChildCommponentFromAttachSocket!(
    self:Custom Character Part Anim Instance Object Reference
    Parent:Skeletal Mesh Component Object Reference=@n10
    bIncludeAllDescendants:bool=false
    AttachSocket:Name=Cape
   ) -> (ReturnValue:Skeletal Mesh Component Object Reference)
   n12 = Get Distance Between Two Sockets!(
    Component:Skeletal Mesh Component Object Reference=@n11
    SocketOrBoneNameA:Name=backpack_root
    SocketSpaceA:ERelativeTransformSpace Enum=RTS_World
    SocketOrBoneNameB:Name=cape_length
    SocketSpaceB:ERelativeTransformSpace Enum=RTS_World
    bRemapRange:bool=false
    InRangeMin:float=0
    InRangeMax:float=0
    OutRangeMin:float=0
    OutRangeMax:float=0
   ) -> (ReturnValue:float) {is=title K2Node_CallFunction}
   > Set isWearingCapeAlpha(isWearingCapeAlpha:double="To Float (Boolean)"(InBool="float > float"(A=@n12, B=50))) -> (Output_Get:double)
   "Not Equal (Object)"!(A:Object Reference=@n11, B:Object Reference) -> (ReturnValue:bool)
   "Not Equal (Object)"(A=FindChildCommponentFromAttachSocket(Parent=@n10, bIncludeAllDescendants=false, AttachSocket=backpack_br))
   n13 = BreakTransform(InTransform=GetDeltaTransformFromRefPose(self=GetOwnerSkeletalMeshComponent(), BoneName=upperarm_r, BaseName=clavicle_r))
   Set LookUp_L(LookUp_L:double=MapRangeClamped(Value=BreakRotator(InRot=@n13.Rotation).Yaw, InRangeA=90, InRangeB=120)) -> (Output_Get:double)
  ]}
]
variables [
 finalGravity {type=Vector}
 headSwingAlpha {type=number}
 isWearingBackblingAlpha {type=number}
 isWearingCapeAlpha {type=number}
 Jump_Falling {type=boolean}
 LookUp_L {type=number}
 windAmbientVector {type=Vector}
 windVector {type=Vector}
]
