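# Physics asset for the Moss CarbideBlack medium male player body, setting up physical secondary motion for foliage pieces on the head.
# A head capsule anchors five small master capsules, each carrying a thin box body for one foliage piece (a through d, plus forehead).
# Master-to-head constraints leave swing and twist free, while piece-to-master constraints restrict them to narrow swing and very tight twist, forming short two-link chains.
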PhysicsAsset /BRCosmetics/Characters/Player/Male/Medium/Bodies/M_MED_Moss_CarbideBlack/Meshes/Parts/M_MED_Moss_CarbideBlack_Physics
bodies {
 dyn_folliage_a [{shapeType=Box, extentX=1.01, extentY=4.0951, extentZ=3.5534, center=<-0.7534 0.0001 0>}]
 dyn_folliage_a_MSTR3 [{shapeType=Capsule, length=0.505, radius=0.505}]
 dyn_folliage_b [{shapeType=Box, extentX=1.01, extentY=4.0951, extentZ=3.5534, center=<-0.7534 0.0001 0>}]
 dyn_folliage_b_MSTR [{shapeType=Capsule, length=0.505, radius=0.505}]
 dyn_folliage_c [{shapeType=Box, extentX=1.01, extentY=4.0951, extentZ=3.5534, center=<-0.7534 0.0001 0>}]
 dyn_folliage_c_MSTR [{shapeType=Capsule, length=0.505, radius=0.505}]
 dyn_folliage_d [{shapeType=Box, extentX=1.01, extentY=4.0951, extentZ=3.5534, center=<-0.7534 0.0001 0>}]
 dyn_folliage_d_MSTR [{shapeType=Capsule, length=0.505, radius=0.505}]
 dyn_folliage_forehead [{shapeType=Box, extentX=1.01, extentY=4.0951, extentZ=3.5534, center=<-1.4981 0.0001 0>}]
 dyn_folliage_forehead_MSTR [{shapeType=Capsule, length=0.505, radius=0.505}]
 head [{shapeType=Capsule, length=8.8005, radius=6.0284, rotation=<rot 90 0 0>}]
}
constraints [
 {
  bone1Name=dyn_folliage_a_MSTR3
  bone2Name=head
  swing1LimitDegrees=45
  swing1Motion=Free
  swing2LimitDegrees=45
  swing2Motion=Free
  twistLimitDegrees=45
  twistMotion=Free
 }
 {
  bone1Name=dyn_folliage_b_MSTR
  bone2Name=head
  swing1LimitDegrees=45
  swing1Motion=Free
  swing2LimitDegrees=45
  swing2Motion=Free
  twistLimitDegrees=45
  twistMotion=Free
 }
 {
  bone1Name=dyn_folliage_c_MSTR
  bone2Name=head
  swing1LimitDegrees=45
  swing1Motion=Free
  swing2LimitDegrees=45
  swing2Motion=Free
  twistLimitDegrees=45
  twistMotion=Free
 }
 {
  bone1Name=dyn_folliage_d_MSTR
  bone2Name=head
  swing1LimitDegrees=45
  swing1Motion=Free
  swing2LimitDegrees=45
  swing2Motion=Free
  twistLimitDegrees=45
  twistMotion=Free
 }
 {
  bone1Name=dyn_folliage_forehead_MSTR
  bone2Name=head
  swing1LimitDegrees=45
  swing1Motion=Free
  swing2LimitDegrees=45
  swing2Motion=Free
  twistLimitDegrees=45
  twistMotion=Free
 }
 {bone1Name=dyn_folliage_a, bone2Name=dyn_folliage_a_MSTR3, swing1LimitDegrees=25, swing2LimitDegrees=25, twistLimitDegrees=5}
 {bone1Name=dyn_folliage_b, bone2Name=dyn_folliage_b_MSTR, swing1LimitDegrees=25, swing2LimitDegrees=25, twistLimitDegrees=5}
 {bone1Name=dyn_folliage_c, bone2Name=dyn_folliage_c_MSTR, swing1LimitDegrees=25, swing2LimitDegrees=25, twistLimitDegrees=5}
 {bone1Name=dyn_folliage_d, bone2Name=dyn_folliage_d_MSTR, swing1LimitDegrees=25, swing2LimitDegrees=25, twistLimitDegrees=5}
 {bone1Name=dyn_folliage_forehead, bone2Name=dyn_folliage_forehead_MSTR, swing1LimitDegrees=25, swing2LimitDegrees=25, twistLimitDegrees=5}
]
