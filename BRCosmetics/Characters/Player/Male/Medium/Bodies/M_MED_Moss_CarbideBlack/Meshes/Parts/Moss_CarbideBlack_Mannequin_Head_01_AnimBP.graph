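# Animation Blueprint for the head mesh part of the Moss CarbideBlack character cosmetic, on the medium male mannequin head skeleton.
# The AnimGraph copies the pose from the attached parent mesh, then runs it through the M_MED_Moss_CarbideBlack Control Rig to the output.
# The EventGraph holds leftover logic mapping head yaw relative to spine_05, but its result is unused and the update event is empty.
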AnimBlueprint /BRCosmetics/Characters/Player/Male/Medium/Bodies/M_MED_Moss_CarbideBlack/Meshes/Parts/Moss_CarbideBlack_Mannequin_Head_01_AnimBP
target_skeleton=/Game/Characters/Player/Male/Medium/Heads/M_Med_Mannequin/Meshes/M_MED_AUX_Mannequin_Head_01Skeleton
graphs [
 AnimGraph {nodes [
   n1 = Control Rig(Source=Copy Pose From Mesh() {bUseAttachedParent=true}) {
    alphaBoolBlend {blendOption=Linear}
    controlRigAssetReference {blueprintRigClass=&/M_MED_Moss_CarbideBlack_CTRL_Rig.M_MED_Moss_CarbideBlack_CTRL_Rig_C}
   }
   Output Pose(Result=@n1) {layerGroup=None}
  ]}
 EventGraph {nodes [
   TryGetPawnOwner()
   n1 = GetDeltaTransformFromRefPose(self=GetOwnerSkeletalMeshComponent(), BoneName=head, BaseName=spine_05)
   MapRangeClamped(Value=BreakRotator(InRot=BreakTransform(InTransform=@n1).Rotation).Yaw, InRangeA=-20, InRangeB=-40)
   Event BlueprintUpdateAnimation()
  ]}
]
variables []
